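; BP_AnimBatchDup_SubItem is a Blueprint data item for an animation batch-duplication editor tool, holding one asset's source path, source name and target name.
; Update Target Name sets the target name directly, while ReplaceName derives it from the unchanged source name by case-sensitive substitution.
; GetSubfolder extracts the immediate parent folder name from the source full path through chained string splits and removal of the source name.
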(event Custom|Update%20Target%20Name
  (params (New%20Name :type String))
  (Variables|Default|SetTarget%20Name New%20Name))

(fn GetSubfolder
  (bind (_lefts _rights) (Utilities|String|Split (Utilities|String|Split (Utilities|String|Replace (Utilities|String|Split (Variables|Default|GetSource%20Full%20Path) "." "CaseSensitive" "FromEnd") (Variables|Default|GetSource%20Name) "" "CaseSensitive") "/" "CaseSensitive" "FromEnd") "/" "CaseSensitive" "FromEnd"))
  (return _rights))

(fn ReplaceName
  (params (From :type String) (To :type String))
  (Variables|Default|SetTarget%20Name (Utilities|String|Replace (Variables|Default|GetSource%20Name) From To "CaseSensitive")))
