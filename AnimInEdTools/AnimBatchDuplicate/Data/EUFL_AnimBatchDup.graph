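# Editor function library for the AnimBatchDuplicate tool, with two lookups for assets related to animation sequences.
# Get AnimMontages from AnimSequence queries Asset Registry referencers of the sequence's package, soft references included, loads each and returns those that cast to AnimMontage.
# Get Linked AnimSeq from LevelSeq reads the Level Sequence's anim sequence links and returns the AnimSequences that exist at the linked paths.
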EditorUtilityBlueprint /AnimInEdTools/AnimBatchDuplicate/Data/EUFL_AnimBatchDup
graphs [
 Get AnimMontages from AnimSequence {nodes [
   Get AnimMontages from AnimSequence!() -> (then, Anim Sequence:Anim Sequence Object Reference) {is=title K2Node_FunctionEntry}
   > n1 = Sequence()
   n2 = GetPackageForObject!(Object:Object Reference=Get Anim Sequence() -> (Anim Sequence:Anim Sequence Object Reference)) -> (ReturnValue:Package Object Reference)
   > n3 = Get Referencers!(
    execute
    self:Asset Registry Interface=GetAssetRegistry()
    PackageName:Name=String To Name(InString=GetDisplayName(Object=@n2))
    ReferenceOptions:"Asset Registry Dependency Options Structure (by ref)"=Make Asset Registry Dependency Options(bIncludeSoftPackageReferences=True)
   ) -> (then, OutReferencers:Array of Names, ReturnValue:bool) {is=title K2Node_CallFunction}
   > n4 = For Each Loop(Array:Array of Names=@n3) -> (Array Element:Name)
   > n5 = LoadAsset(AssetPath="To String (Name)"(InName=@n4))
   > n6 = Cast To AnimMontage!(execute, Object=@n5) -> (then, CastFailed, AsAnim Montage:Anim Montage Object Reference) {is=title K2Node_DynamicCast}
   > Add(
    TargetArray:Array of Anim Montage Object References=Get Local Anim Montages() -> (Local Anim Montages:Array of Anim Montage Object References)
    NewItem:"Anim Montage Object Reference (by ref)"=@n6
   ) {is=title K2Node_CallArrayFunction}
   Return Node!(execute=@n1.then_1, Anim Montages:Array of Anim Montage Object References=Get Local Anim Montages())
  ]}
 Get Linked AnimSeq from LevelSeq {nodes [
   Get Linked AnimSeq from LevelSeq!() -> (then, Level Sequence:Level Sequence Object Reference) {is=title K2Node_FunctionEntry}
   > n1 = Sequence()
   > n2 = GetAnimSequenceLinkFromLevelSequence!(execute, InLevelSequence:Level Sequence Object Reference=Get Level Sequence() -> (Level Sequence:Level Sequence Object Reference)) -> (then, ReturnValue:Level Sequence Anim Sequence Link Object Reference)
   > n3 = For Each Loop(Array:Array of Level Sequence Anim Sequence Link Item Structures=Get AnimSequenceLinks!(self:Level Sequence Anim Sequence Link Object Reference=@n2) -> (AnimSequenceLinks:Array of Level Sequence Anim Sequence Link Item Structures)) -> (Array Element:Level Sequence Anim Sequence Link Item Structure)
   n4 = Break Level Sequence Anim Sequence Link Item!(LevelSequenceAnimSequenceLinkItem:"Level Sequence Anim Sequence Link Item Structure (by ref)"=@n3) -> (
    SkelTrackGuid:Guid Structure
    PathToAnimSequence_PathString:String
    bAutoBake:bool=~
    bExportTransforms:bool=~
    bExportMorphTargets:bool=~
    bExportAttributeCurves:bool=~
    bExportMaterialCurves:bool=~
    bSkipCurvesWithZeroValue:bool=~
    bSetRetargetSourceAsset:bool=~
    Interpolation:EAnimInterpolationType Enum
    CurveInterpolation:ERichCurveInterpMode Enum
    bRecordInWorldSpace:bool=~
    bEvaluateAllSkeletalMeshComponents:bool=~
    IncludeAnimationNames:Array of Strings
    ExcludeAnimationNames:Array of Strings
    bRemoveExcludedCurves:bool=~
    WarmUpFrames:Frame Number Structure
    DelayBeforeStart:Frame Number Structure
    bUseCustomTimeRange:bool=~
    CustomStartFrame:Frame Number Structure
    CustomEndFrame:Frame Number Structure
    CustomDisplayRate:Frame Rate Structure
    bUseCustomFrameRate:bool=~
    CustomFrameRate:Frame Rate Structure
   ) {is=title K2Node_BreakStruct}
   > n5 = DoesAssetExist(AssetPath=@n4.PathToAnimSequence_PathString)
   > Branch(Condition=@n5)
   > n6 = LoadAsset(AssetPath=@n4.PathToAnimSequence_PathString)
   > n7 = Cast To AnimSequence!(execute, Object=@n6) -> (then, CastFailed, AsAnim Sequence:Anim Sequence Object Reference)
   > Add(
    TargetArray:Array of Anim Sequence Object References=Get Local AnimSeq() -> (Local AnimSeq:Array of Anim Sequence Object References)
    NewItem:"Anim Sequence Object Reference (by ref)"=@n7
   )
   Return Node!(execute=@n1.then_1, Animation Sequences:Array of Anim Sequence Object References=Get Local AnimSeq())
  ]}
]
variables []
